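# Physics asset for the BudBlossom backpack cosmetic, defining capsule bodies and joints for its dyn_ pack, handle, chain and rose bones.
# Two handles, a five-link chain and a branching rose stem with two branches and a leaf all hang off dyn_pack.
# Chain root and handle tips are locked to their parents.
# Handle bases swing on one axis, while other chain and rose joints allow only small angular movement.
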PhysicsAsset /BRCosmetics/Accessories/FORT_Backpacks/Backpack_F_MED_BudBlossom/Meshes/F_MED_BudBlossom_Pack_Physics
bodies {
 dyn_chain_01 [{shapeType=Capsule, length=0.505, radius=0.505, shapeName=dyn_chain_01_capsule}]
 dyn_chain_02 [{
   shapeType=Capsule
   length=1.4345
   radius=0.7455
   shapeName=dyn_chain_04_capsule
   center=<1.0453 -0.0013 0.0001>
   rotation=<rot 0 89.9943 -89.9942>
  }]
 dyn_chain_03 [{
   shapeType=Capsule
   length=1.4345
   radius=0.7455
   shapeName=dyn_chain_04_capsule
   center=<1.0453 -0.0013 0.0001>
   rotation=<rot 0 89.9943 -89.9942>
  }]
 dyn_chain_04 [{
   shapeType=Capsule
   length=1.4345
   radius=0.7455
   shapeName=dyn_chain_04_capsule
   center=<1.0453 -0.0013 0.0001>
   rotation=<rot 0 89.9943 -89.9942>
  }]
 dyn_chain_05 [{
   shapeType=Capsule
   length=2.4981
   radius=1.5066
   shapeName=dyn_chain_05_capsule
   center=<1.5588 -0.059 0.02>
   rotation=<rot -0.7373 179.4384 -89.9306>
  }]
 dyn_handle_bk_01 [{
   shapeType=Capsule
   length=3.5188
   radius=0.7553
   shapeName=dyn_handle_bk_01_capsule
   center=<2.4051 -0.0036 0.0366>
   rotation=<rot -0.2796 90.0495 -90.0002>
  }]
 dyn_handle_bk_02 [{
   shapeType=Capsule
   length=1.2308
   radius=0.9465
   shapeName=dyn_handle_bk_02_capsule
   center=<-0.5171 0 -0.1686>
   rotation=<rot 90 -90.0001 0>
  }]
 dyn_handle_fwd_01 [{
   shapeType=Capsule
   length=3.5188
   radius=0.7553
   shapeName=dyn_handle_bk_01_capsule
   center=<2.4051 -0.0036 0.0366>
   rotation=<rot -0.2796 90.0495 -90.0002>
  }]
 dyn_handle_fwd_02 [{shapeType=Capsule, length=1.2308, radius=0.9465, shapeName=dyn_handle_bk_02_capsule, rotation=<rot 90 -90.0001 0>}]
 dyn_pack [
  {shapeType=Capsule, length=4.6214, radius=3.5251, center=<-5.1169 0.6562 -2.4779>, rotation=<rot 82.1783 -140.35 -140.1759>}
  {shapeType=Capsule, length=10.3169, radius=5.9031, shapeName=dyn_pack_capsule, center=<3.4448 0 2.7601>, rotation=<rot 6.1568 0 -0.0002>}
 ]
 dyn_rose_01 [{
   shapeType=Capsule
   length=1.3848
   radius=0.505
   shapeName=dyn_rose_01_capsule
   center=<0.8153 -0.0766 0.052>
   rotation=<rot 85 -0.0001 -0.0003>
  }]
 dyn_rose_A_01 [{shapeType=Capsule, length=2.5818, radius=0.3, shapeName=dyn_rose_A_01_capsule, rotation=<rot 90 0 0>}]
 dyn_rose_A_02 [{shapeType=Capsule, length=3.3024, radius=0.5, shapeName=dyn_rose_A_01_capsule, center=<0.9122 0 0.0619>, rotation=<rot 90 0 0>}]
 dyn_rose_A_03 [{
   shapeType=Capsule
   length=2.0441
   radius=2.4428
   shapeName=dyn_rose_A_03_capsule
   center=<3.7952 0 -0.1741>
   rotation=<rot 70.0002 -0.0003 -0.0007>
  }]
 dyn_rose_B_01 [{
   shapeType=Capsule
   length=3.3804
   radius=0.3
   shapeName=dyn_rose_A_01_capsule
   center=<-0.7781 0 -0.1425>
   rotation=<rot 85.9994 -179.9998 179.9999>
  }]
 dyn_rose_B_02 [{
   shapeType=Capsule
   length=2.0441
   radius=2.4428
   shapeName=dyn_rose_A_03_capsule
   center=<2.0604 0.2022 0.128>
   rotation=<rot 80 -179.9989 180>
  }]
 dyn_rose_leaf_A_01 [{
   shapeType=Capsule
   length=1.5488
   radius=0.9726
   shapeName=dyn_rose_leaf_A_01_capsule
   center=<1.8621 -0.1198 0.109>
   rotation=<rot 85 -0.0003 -0.0005>
  }]
}
constraints [
 {bone1Name=dyn_handle_bk_01, bone2Name=dyn_pack, swing1LimitDegrees=0, swing2LimitDegrees=8, twistLimitDegrees=0}
 {bone1Name=dyn_handle_fwd_01, bone2Name=dyn_pack, swing1LimitDegrees=0, swing2LimitDegrees=15, twistLimitDegrees=0}
 {
  bone1Name=dyn_handle_bk_02
  bone2Name=dyn_handle_bk_01
  swing1LimitDegrees=45
  swing1Motion=Locked
  swing2LimitDegrees=45
  swing2Motion=Locked
  twistLimitDegrees=45
  twistMotion=Locked
 }
 {
  bone1Name=dyn_handle_fwd_02
  bone2Name=dyn_handle_fwd_01
  swing1LimitDegrees=45
  swing1Motion=Locked
  swing2LimitDegrees=45
  swing2Motion=Locked
  twistLimitDegrees=45
  twistMotion=Locked
 }
 {
  bone1Name=dyn_chain_01
  bone2Name=dyn_pack
  swing1LimitDegrees=45
  swing1Motion=Locked
  swing2LimitDegrees=45
  swing2Motion=Locked
  twistLimitDegrees=45
  twistMotion=Locked
 }
 {bone1Name=dyn_chain_02, bone2Name=dyn_chain_01, swing1LimitDegrees=5, swing2LimitDegrees=10, twistLimitDegrees=1}
 {bone1Name=dyn_chain_03, bone2Name=dyn_chain_02, swing1LimitDegrees=5, swing2LimitDegrees=10, twistLimitDegrees=1}
 {bone1Name=dyn_chain_04, bone2Name=dyn_chain_03, swing1LimitDegrees=5, swing2LimitDegrees=10, twistLimitDegrees=1}
 {bone1Name=dyn_chain_05, bone2Name=dyn_chain_04, swing1LimitDegrees=8, swing2LimitDegrees=10, twistLimitDegrees=2}
 {bone1Name=dyn_rose_01, bone2Name=dyn_pack, swing1LimitDegrees=1, swing2LimitDegrees=2, twistLimitDegrees=1}
 {bone1Name=dyn_rose_A_01, bone2Name=dyn_rose_01, swing1LimitDegrees=1, swing2LimitDegrees=3, twistLimitDegrees=0}
 {bone1Name=dyn_rose_A_02, bone2Name=dyn_rose_A_01, swing1LimitDegrees=1, swing2LimitDegrees=3, twistLimitDegrees=0}
 {bone1Name=dyn_rose_A_03, bone2Name=dyn_rose_A_02, swing1LimitDegrees=1, swing2LimitDegrees=2, twistLimitDegrees=0}
 {bone1Name=dyn_rose_leaf_A_01, bone2Name=dyn_rose_A_02, swing1LimitDegrees=2, swing2LimitDegrees=4, twistLimitDegrees=0}
 {bone1Name=dyn_rose_B_01, bone2Name=dyn_rose_01, swing1LimitDegrees=1, swing2LimitDegrees=3, twistLimitDegrees=0}
 {bone1Name=dyn_rose_B_02, bone2Name=dyn_rose_B_01, swing1LimitDegrees=1, swing2LimitDegrees=2, twistLimitDegrees=0}
]
