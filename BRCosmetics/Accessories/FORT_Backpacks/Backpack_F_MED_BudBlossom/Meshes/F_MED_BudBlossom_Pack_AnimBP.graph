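# Animation Blueprint for the BudBlossom backpack cosmetic, adding physics-driven secondary motion.
# The input pose runs through a RigidBody simulation in pack-bone space, then fixed additive bone offsets on the pack, dynamic chain and handles.
# Each update flips the simulation's gravity override upward during skydiving, sprinting, sliding or parachuting, otherwise restoring downward gravity.
# Skydive and jump floats are also set but unread here, and some transform nodes are disconnected.
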AnimBlueprint /BRCosmetics/Accessories/FORT_Backpacks/Backpack_F_MED_BudBlossom/Meshes/F_MED_BudBlossom_Pack_AnimBP
target_skeleton=&/F_MED_BudBlossom_Pack_Skeleton
graphs [
 AnimGraph {nodes [
   "Transform (Modify) Bone - Bone: base_handle_fwd_01"(Alpha=1, Translation="0,0,0", Rotation="90,0,0", Scale="1,1,1") {
    rotationMode=BMM_Additive
    rotationSpace=BCS_BoneSpace
    translationMode=BMM_Additive
    translationSpace=BCS_BoneSpace
    alphaScaleBiasClamp {bInterpResult=true, interpSpeedDecreasing=5, interpSpeedIncreasing=5}
    boneToModify {boneName=base_handle_fwd_01}
    rotation {pitch=90}
   }
   n1 = "Transform (Modify) Bone - Bone: dyn_chain_02"(Alpha=1, Translation="0,0,0", Rotation="45,0,0", Scale="1,1,1") {
    rotationMode=BMM_Additive
    rotationSpace=BCS_BoneSpace
    translationMode=BMM_Additive
    translationSpace=BCS_BoneSpace
    alphaScaleBiasClamp {bInterpResult=true, interpSpeedDecreasing=5, interpSpeedIncreasing=5}
    boneToModify {boneName=dyn_chain_02}
    rotation {pitch=45}
   }
   n2 = "Transform (Modify) Bone - Bone: dyn_chain_03"(ComponentPose=@n1, Alpha=1, Translation="0,0,0", Rotation="25,0,0", Scale="1,1,1") {
    rotationMode=BMM_Additive
    rotationSpace=BCS_BoneSpace
    translationMode=BMM_Additive
    translationSpace=BCS_BoneSpace
    alphaScaleBiasClamp {bInterpResult=true, interpSpeedDecreasing=5, interpSpeedIncreasing=5}
    boneToModify {boneName=dyn_chain_03}
    rotation {pitch=25}
   }
   n3 = "Transform (Modify) Bone - Bone: dyn_chain_04"(ComponentPose=@n2, Alpha=1, Translation="0,0,0", Rotation="30,0,0", Scale="1,1,1") {
    rotationMode=BMM_Additive
    rotationSpace=BCS_BoneSpace
    translationMode=BMM_Additive
    translationSpace=BCS_BoneSpace
    alphaScaleBiasClamp {bInterpResult=true, interpSpeedDecreasing=5, interpSpeedIncreasing=5}
    boneToModify {boneName=dyn_chain_04}
    rotation {pitch=30}
   }
   "Transform (Modify) Bone - Bone: dyn_chain_05"(ComponentPose=@n3, Alpha=1, Translation="0,0,0", Rotation="15,0,0", Scale="1,1,1") {
    rotationMode=BMM_Additive
    rotationSpace=BCS_BoneSpace
    translationMode=BMM_Additive
    translationSpace=BCS_BoneSpace
    alphaScaleBiasClamp {bInterpResult=true, interpSpeedDecreasing=5, interpSpeedIncreasing=5}
    boneToModify {boneName=dyn_chain_05}
    rotation {pitch=15}
   }
   n4 = RigidBody!(
    ComponentPose=Local To Component(LocalPose=Input Pose())
    Alpha
    OverrideWorldGravity:Vector=Get Gravity_Override() -> (Gravity_Override:Vector)
    ExternalForce
   ) -> (Pose) {
    bOverrideWorldGravity=true
    overridePhysicsAsset=&/F_MED_BudBlossom_Pack_Physics
    simulationSpace=BaseBoneSpace
    baseBoneRef {boneName=pack}
    overrideWorldGravity {z=-980}
    simSpaceSettings {maxAngularAcceleration=3500, maxAngularVelocity=700, maxLinearAcceleration=8, maxLinearVelocity=40, worldAlpha=1}
   }
   n5 = "Transform (Modify) Bone - Bone: pack"(ComponentPose=@n4, Translation="-0.005,-1.324,-0.239", Rotation="0.529,2.946,-10.172") {
    rotationMode=BMM_Additive
    rotationSpace=BCS_BoneSpace
    translationMode=BMM_Additive
    translationSpace=BCS_BoneSpace
    alphaScaleBiasClamp {bInterpResult=true, interpSpeedDecreasing=5, interpSpeedIncreasing=5}
    boneToModify {boneName=pack}
    rotation=<rot 0.529 2.946 -10.1721>
    translation=<-0.005 -1.3243 -0.2391>
   }
   n6 = "Transform (Modify) Bone - Bone: dyn_chain_02"(ComponentPose=@n5, Alpha=1, Translation="0,0,0", Rotation="75,-0.001,-0.002", Scale="1,1,1") {
    rotationMode=BMM_Additive
    rotationSpace=BCS_BoneSpace
    translationMode=BMM_Additive
    translationSpace=BCS_BoneSpace
    alphaScaleBiasClamp {bInterpResult=true, interpSpeedDecreasing=5, interpSpeedIncreasing=5}
    boneToModify {boneName=dyn_chain_02}
    rotation=<rot 75 -0.001 -0.002>
   }
   n7 = "Transform (Modify) Bone - Bone: dyn_chain_03"(ComponentPose=@n6, Alpha=1, Translation="0,0,0", Rotation="30,0,0", Scale="1,1,1") {
    rotationMode=BMM_Additive
    rotationSpace=BCS_BoneSpace
    translationMode=BMM_Additive
    translationSpace=BCS_BoneSpace
    alphaScaleBiasClamp {bInterpResult=true, interpSpeedDecreasing=5, interpSpeedIncreasing=5}
    boneToModify {boneName=dyn_chain_03}
    rotation {pitch=30}
   }
   n8 = "Transform (Modify) Bone - Bone: dyn_chain_04"(ComponentPose=@n7, Alpha=1, Translation="0,0,0", Rotation="15,0,0", Scale="1,1,1") {
    rotationMode=BMM_Additive
    rotationSpace=BCS_BoneSpace
    translationMode=BMM_Additive
    translationSpace=BCS_BoneSpace
    alphaScaleBiasClamp {bInterpResult=true, interpSpeedDecreasing=5, interpSpeedIncreasing=5}
    boneToModify {boneName=dyn_chain_04}
    rotation {pitch=15}
   }
   n9 = "Transform (Modify) Bone - Bone: dyn_chain_05"(ComponentPose=@n8, Alpha=1, Translation="0,0,0", Rotation="10,0,0", Scale="1,1,1") {
    rotationMode=BMM_Additive
    rotationSpace=BCS_BoneSpace
    translationMode=BMM_Additive
    translationSpace=BCS_BoneSpace
    alphaScaleBiasClamp {bInterpResult=true, interpSpeedDecreasing=5, interpSpeedIncreasing=5}
    boneToModify {boneName=dyn_chain_05}
    rotation {pitch=10}
   }
   n10 = "Transform (Modify) Bone - Bone: base_handle_fwd_01"(ComponentPose=@n9, Alpha=1, Translation="0,0,0", Rotation="60,0,0", Scale="1,1,1") {
    rotationMode=BMM_Additive
    rotationSpace=BCS_BoneSpace
    translationMode=BMM_Additive
    translationSpace=BCS_BoneSpace
    alphaScaleBiasClamp {bInterpResult=true, interpSpeedDecreasing=5, interpSpeedIncreasing=5}
    boneToModify {boneName=base_handle_fwd_01}
    rotation {pitch=60}
   }
   n11 = "Transform (Modify) Bone - Bone: base_handle_bk_01"(ComponentPose=@n10, Alpha=1, Translation="0,0,0", Rotation="15,0,0", Scale="1,1,1") {
    rotationMode=BMM_Additive
    rotationSpace=BCS_BoneSpace
    translationMode=BMM_Additive
    translationSpace=BCS_BoneSpace
    alphaScaleBiasClamp {bInterpResult=true, interpSpeedDecreasing=5, interpSpeedIncreasing=5}
    boneToModify {boneName=base_handle_bk_01}
    rotation {pitch=15}
   }
   n12 = "Transform (Modify) Bone - Bone: base_handle_fwd_01"(ComponentPose=@n11, Alpha=1, Translation="0,0,0", Rotation="60,0,0", Scale="1,1,1") {
    rotationMode=BMM_Additive
    rotationSpace=BCS_BoneSpace
    translationMode=BMM_Additive
    translationSpace=BCS_BoneSpace
    alphaScaleBiasClamp {bInterpResult=true, interpSpeedDecreasing=5, interpSpeedIncreasing=5}
    boneToModify {boneName=base_handle_fwd_01}
    rotation {pitch=60}
   }
   n13 = "Transform (Modify) Bone - Bone: base_handle_bk_01"(ComponentPose=@n12, Alpha=1, Translation="0,0,0", Rotation="15,0,0", Scale="1,1,1") {
    rotationMode=BMM_Additive
    rotationSpace=BCS_BoneSpace
    translationMode=BMM_Additive
    translationSpace=BCS_BoneSpace
    alphaScaleBiasClamp {bInterpResult=true, interpSpeedDecreasing=5, interpSpeedIncreasing=5}
    boneToModify {boneName=base_handle_bk_01}
    rotation {pitch=15}
   }
   Output Pose(Result=Component To Local(ComponentPose=@n13)) {layerGroup=None}
  ]}
 EventGraph {nodes [
   Event BlueprintUpdateAnimation()
   > n1 = Sequence()
   n2 = Get bIsSkydiveFloating()
   n3 = Get bIsSkydiveDiveMode()
   > Set SkyDive_Floating(SkyDive_Floating:double="To Float (Boolean)"(InBool=OR Boolean(A=@n2, B=@n3))) -> (Output_Get:double)
   n4 = Get bIsSkyDiving()
   > n5 = Branch(Condition=AND Boolean(A=Get bIsFalling(), B=NOT Boolean(A=@n4)))
   > Set Jumping_Float(Jumping_Float:double=1) -> (Output_Get:double)
   n6 = Branch(execute=@n5.else, Condition=Get bIsJumping())
   > Set Jumping_Float(Jumping_Float=0.6)
   Set Jumping_Float(execute=@n6.else, Jumping_Float=0)
   n7 = Branch(
    execute=@n1.then_1
    Condition=OR Boolean!(
     A=Get bIsSkydiveFloating()
     B=Get bIsSkydiveDiveMode()
     C:bool=Get bIsSprinting()
     D:bool=Get bIsKneeSliding()
     E:bool=Get bIsBaseballSliding()
     F:bool=Get bIsParachuteOpened()
    ) -> (ReturnValue)
   )
   > Set Gravity_Override(Gravity_Override:Vector="0,0,450") -> (Output_Get:Vector)
   Set Gravity_Override(execute=@n7.else, Gravity_Override="0,0,-980")
   TryGetPawnOwner()
  ]}
]
variables [
 Gravity_Override {type=Vector, default=<0 0 -980>}
 Jumping_Float {type=number}
 Jumping_half_Float {type=number}
 SkyDive_Floating {type=number}
]
